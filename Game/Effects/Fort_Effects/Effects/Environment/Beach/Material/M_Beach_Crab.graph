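# Masked, two-sided lit material for beach crab Niagara mesh particles, animating legs and body through vertex math.
# Leg motion is a time-driven sine phased by texture alpha, masked by that alpha and UV1 height, then scaled by Leg WPO.
# Two sine-driven RotateAboutAxis wobbles around the component pivot add to it, with amounts from particle dynamic parameters.
# Texture color times particle color is split between base color and emissive contribution scalars.
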Material /Game/Effects/Fort_Effects/Effects/Environment/Beach/Material/M_Beach_Crab
graph {
 properties {
  bTangentSpaceNormal=false
  bUsedWithNiagaraMeshParticles=true
  blendMode=BLEND_Masked
  opacityMaskClipValue=0.333
  pixelDepthOffsetMode=PDOM_Legacy
  twoSided=true
  usedShadingModels=Default Lit
 }
 nodes [
  n1 = TextureSample() {texture=/Game/Effects/Fort_Effects/Effects/Environment/Beach/Texture/T_Beach_Crab_BaseColor}
  n2 = Add(A=Time(), B=NamedRerouteUsage())
  n3 = DynamicParameter() {defaultValue=<rgba 1 1 4 0.05>, paramNames [SineWavelength, SineAmplitude, LegMask, RotationX]}
  n4 = Sine(None=Multiply(A=Add(A=@n1.A, B=@n2), B=@n3.SineWavelength))
  n5 = Multiply(A=LinearInterpolate(Alpha=ConstantBiasScale(None=@n4)) {constA=-1}, B=@n3.SineAmplitude)
  n6 = Power(Base=ComponentMask(None=TextureCoordinate() {coordinateIndex=1}) {g=true}, Exp=@n3.LegMask) {constExponent=4}
  n7 = Saturate(None=Multiply(A=OneMinus(None=@n1.A)) {constB=10})
  n8 = VectorParameter() {parameterName=Leg WPO, defaultValue=<rgba 3 5 0 0>}
  n9 = Normalize(VectorInput=VectorParameter() {parameterName=Param_2, defaultValue=<rgba 1 0 0 1>}.RGB)
  n10 = Multiply(
   A=DynamicParameter() {defaultValue=<rgba 1 1 4 0.05>, paramNames [SineWavelength, SineAmplitude, LegMask, RotationX]}.RotationX
   B=MaterialFunctionCall(Speed=Constant() {r=1}) {materialFunction=&DebugTimeSine}.Result
  ) {constA=-0.5}
  n11 = MaterialFunctionCall() {materialFunction=&ComponentPivotLocation}
  n12 = WorldPosition()
  n13 = Normalize(VectorInput=Constant3Vector() {constant=<rgba 0 0 1 0>})
  n14 = Multiply(
   A=DynamicParameter() {parameterIndex=1, defaultValue=<rgba 0.05 1 1 1>, paramNames [RotationZ,Param2,Param3,Param4]}.RotationZ
   B=MaterialFunctionCall(Speed=Constant() {r=1}) {materialFunction=&DebugTimeSine}.Result
  ) {constA=-0.5}
  n15 = MaterialFunctionCall() {materialFunction=&ComponentPivotLocation}
  n16 = WorldPosition()
  Add(
   A=Multiply(A=Multiply(A=@n5, B=@n6), B=Multiply(A=@n7, B=@n8.RGB))
   B=Add(
    A=RotateAboutAxis(NormalizedRotationAxis=@n9, RotationAngle=@n10, PivotPoint=@n11.Result, Position=@n12.XYZ)
    B=RotateAboutAxis(NormalizedRotationAxis=@n13, RotationAngle=@n14, PivotPoint=@n15.Result, Position=@n16.XYZ)
   )
  )
  Constant()
  Constant()
  Constant() {r=0.5}
  n17 = NamedRerouteUsage()
  Multiply(A=@n17, B=ScalarParameter() {defaultValue=1, parameterName=BaseColor Contrib})
  Multiply(A=@n17, B=ScalarParameter() {parameterName=EmissiveContrib})
  NamedRerouteDeclaration(None=Multiply(A=@n1.RGB, B=ParticleColor().RGB)) {name=BaseColor, nodeColor=<rgba 0.3125 1 0 1>}
  n18 = TransformPosition(None=WorldPosition().XYZ) {transformSourceType=TRANSFORMPOSSOURCE_World, transformType=TRANSFORMPOSSOURCE_Instance}
  n19 = Constant() {r=0.5}
  NamedRerouteDeclaration(None=If(A=ComponentMask(None=@n18) {r=true}, "A > B"=@n19, "A == B"=@n19, "A < B"=Constant())) {name=LeftOrRight, nodeColor=<rgba 1 0.7969 0 1>}
  NamedRerouteDeclaration(None=MaterialFunctionCall!(Width=Constant() {r=0.1}, Sharpness, Offset=Constant() {r=0.4}, Direction Switch=StaticBool() {value=true}, Compare, Input Coordinates) {materialFunction=/Engine/Functions/Engine_MaterialFunctions03/Procedurals/GeneratedBand}.Result) {name=Leg1, nodeColor=<rgba 0.8984 1 0 1>}
  NamedRerouteDeclaration(None=MaterialFunctionCall(Width=Constant() {r=0.1}, Offset=Constant() {r=0.1}, Direction Switch=StaticBool() {value=true}) {materialFunction=&GeneratedBand}.Result) {name=Leg2, nodeColor=<rgba 0 1 0.2365 1>}
 ]
}
