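# Niagara system for an ambient beach seagull flock, drawn as GPU mesh particles with one emitter.
# Spawning follows time of day (none at night) and camera-distance culling, placing birds with sphere and torus locations, the latter sized by User.SpawnRadius.
# Curl noise and a vortex with an oscillating origin steer flight.
# Curves scale birds over life and camera distance, with time-of-day tinting and randomized material parameters.
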NiagaraSystem /Game/Effects/Fort_Effects/Effects/Environment/Beach/NS_Env_Beach_Seagulls
dependencies {
 usedDataInterfaces [&NiagaraDataInterfaceCamera, &NiagaraDataInterfaceCurve]
 usedDynamicInputs []
 usedModules [&SystemState]
}
emitters {BirdsInteractabler {
  simTarget=GPUComputeSim
  emitterUpdateScript [
   /SetVariables(Emitter.CameraPosition:Vector3f=&ReturnCameraVector(Camera Property=::NewEnumerator0, Coordinate Space=::Simulation)) {isSetParameters=true}
   &EmitterState(
    Enable Distance Culling=true
    Life Cycle Mode=::NewEnumerator1
    Loop Behavior=::NewEnumerator1
    Loop Duration Mode=::NewEnumerator1
    MaxDistance=50000
    Min Distance Response=::SleepAndLetParticlesFinish
    MinDistance=5000
    Module.bApplySpawnCountScaleByDistance=true
    Module.bUseMaxDistance=true
    Module.bUseMinDistance=true
    Scalability Mode=::NewEnumerator1
    Scale Spawn Count=true
    Spawn Count Scale By Distance=DataInterface() {propertyValues {Curve {keys [<at 0 0>,<at 0.1 1>,<at 0.95 1>,<at 1 0>]}}}
    UseLoopDelay=false
   )
   &SpawnRate(
    Spawn Probability=&FNTimeOfDayScalar(Night Scale=0, Morning Scale=1, Evening Scale=0.25)
    SpawnRate=10
    Use Spawn Probability=true
   ) {module=SpawnRate001}
  ]
  particleSpawnScript [
   /Niagara/Modules/Spawn/Initialization/InitializeParticle(
    Lifetime=5
    Mesh Scale=&Lerp_Vector(A=<1 1 1>, B=<1.25 1.25 1.25>, Alpha=&UniformAOrBFloat(A=0))
    Module.Write Color=true
    Module.Write Lifetime=true
    Module.Write Mass=true
    Module.Write Material Random=true
    Module.Write Position=false
    Module.Write Scale=true
    Module.Write SpriteRotation=false
    Module.Write SpriteSize=false
    Position=&SimulationPosition()
   )
   &ApplyInitialForces()
   /Niagara/Modules/Spawn/Location/SphereLocation(MaskSpawnGroup=false, Module.Override Randomness=false, Non Uniform Scale=<1 1 0.4>, Sphere Radius=2000, Surface Only=true)
   &DynamicMaterialParameters(
    Index 0 Param 2=&UniformRangedFloat(Minimum=1.25, Maximum=1.5)
    Module.Param0WriteEnabled=true
    Module.Param1WriteEnabled=true
    Module.Param2WriteEnabled=true
    Module.Param3WriteEnabled=true
   )
   &TorusLocation(
    Align Torus With Axis=false
    Enable Torus Knot Radius=false
    Handle Radius=1000
    Large Radius:NiagaraFloat=$User.SpawnRadius
    MaskSpawnGroup=false
    Module.Override Randomness=false
    Torus Mode=::NewEnumerator1
   )
  ]
  particleUpdateScript [
   &ParticleState()
   &CurlNoiseForce(Curl Noise Cone Mask Angle=90, Mask Curl Noise=true, Noise Frequency=6000, Noise Strength=500, Pan Noise=false)
   &VortexForce(
    UseOriginOffset=true
    Vortex Force Amount=75
    Vortex Origin Offset:Vector3f=&Lerp_Vector(
     A=<250 -5000 0>
     B=<-250 5000 0>
     Alpha=&Waveform("Waveform [1]"=::NewEnumerator0, "[1] Frequency"=0.05, "[1] Amplitude Min/Max"=<0 1>)
    )
   ) {module=VortexForce001}
   &SolveForcesAndVelocity(Clamp Velocity=true, Limit Acceleration=true)
   &ScaleMeshSize(Scale Factor=&Multiply_Vector(
     A=<1 1 1>
     B=&VectorFromFloat(Value=&Multiply_Float(
       A=&FloatFromCurve(FloatCurve=DataInterface() {propertyValues {Curve {keys [
            {arriveTangent=1.7794, interpMode=RCIM_Cubic, leaveTangent=1.7794, tangentMode=RCTM_User, time=-0.0063, value=0.4982}
            {arriveTangent=0.3677, interpMode=RCIM_Cubic, leaveTangent=0.3677, tangentMode=RCTM_User, time=0.2892, value=0.9107}
            {arriveTangent=-1.2812, interpMode=RCIM_Cubic, leaveTangent=-1.2812, tangentMode=RCTM_User, time=0.8446, value=0.8923}
            <at 0.9985 0.5165>
           ]}}})
       B=&FloatFromCurve(
        FloatCurve=DataInterface() {propertyValues {Curve {keys [<at 0.2476 0>, {interpMode=RCIM_Cubic, tangentMode=RCTM_User, time=0.3555, value=1.2901}, <at 0.8976 2.0314>, <at 1 0>]}}}
        CurveIndex=&NormalizeDistanceRange(
         Start Position:NiagaraPosition=Linked() {linkedVariable {name=Emitter.CameraPosition, type=Vector3f}}
         End Position:NiagaraPosition=Linked() {linkedVariable {name=Particles.FN_Orbit_v4.Position, type=Vector3f}}
         Distance:NiagaraFloat=15000
         Invert Normalized Range:NiagaraBool=false
        )
        Scale Curve=2
       )
      ))
    )) {enabled=false}
   &ScaleMeshSize(Scale Factor=&Multiply_Vector(
     A=<1 1 1>
     B=&VectorFromFloat(Value=&Multiply_Float(
       A=&FloatFromCurve(
        FloatCurve=DataInterface() {propertyValues {Curve {keys [
            {arriveTangent=1.7794, interpMode=RCIM_Cubic, leaveTangent=1.7794, tangentMode=RCTM_User, time=0, value=0}
            {arriveTangent=0.3677, interpMode=RCIM_Cubic, leaveTangent=0.3677, tangentMode=RCTM_User, time=0.1297, value=0.8851}
            {arriveTangent=-1.2812, interpMode=RCIM_Cubic, leaveTangent=-1.2812, tangentMode=RCTM_User, time=0.8446, value=0.8923}
            <at 0.9985 0>
           ]}}}
        Scale Curve=0.66
       )
       B=&FloatFromCurve(
        FloatCurve=DataInterface() {propertyValues {Curve {keys [<at -0.0055 -0.0388>, {interpMode=RCIM_Cubic, tangentMode=RCTM_User, time=0.0645, value=2}, <at 0.9 2>, <at 1 0>]}}}
        CurveIndex=&NormalizeDistanceRange(
         Start Position:NiagaraPosition=Linked() {linkedVariable {name=Emitter.CameraPosition, type=Vector3f}}
         End Position:NiagaraPosition=$Particles.Position
         Distance:NiagaraFloat=50000
         Invert Normalized Range:NiagaraBool=false
        )
       )
      ))
    )) {module=ScaleMeshSize001}
   /Niagara/Modules/Update/Camera/MaintainInCameraParticleScale() {enabled=false}
   &DynamicMaterialParameters(
    Index 0 Param 2=&RandomRangeFloat(Minimum=2, Maximum=2.5)
    Module.Param0WriteEnabled=true
    Module.Param1WriteEnabled=true
    Module.Param2WriteEnabled=true
    Module.Param3WriteEnabled=true
   ) {module=DynamicMaterialParameters001}
   &ScaleColor(
    Scale RGB=&MakeVectorFromLinearColorRGB(Color=&FN_Select_Color_from_Time_Of_Day(Morning=<rgba 1 0.7288 0.724 1>, Day=<rgba 1 1 1 1>, Evening=<rgba 1 0.7305 0.7231 1>, Night=<rgba 0.6691 0.8446 1 1>))
    ScaleA=true
    ScaleRGB=true
   )
  ]
  rendererClasses [&NiagaraMeshRendererProperties]
  renderers [{rendererClass=&NiagaraMeshRendererProperties, rendererIndex=0}]
 }}
summary {
 systemName=NS_Env_Beach_Seagulls
 emitters [{emitterName=BirdsInteractabler, simTarget=GPUComputeSim, rendererClasses [&NiagaraMeshRendererProperties]}]
 userVariables [User.SpawnRadius {type=NiagaraFloat, defaultValue=2000, description=""}]
}
